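# Animation Blueprint for the Fortnite player skeleton that supplies directional boost jump animations through MovementMode animation layers.
# The InAirOverride layer looks up Direction in DirectionToSequenceMap and plays the forward, backward, left or right sequence, ignoring its input poses.
# The other MovementMode layers have no wired logic.
# The event graph holds only unconnected stub nodes, so nothing here sets Direction, which defaults to North.
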AnimBlueprint /BoostedJump/Animation/ABP_BoostJump
target_skeleton=&Fortnite_M_Avg_Player_Skeleton
graphs [
 AnimGraph {nodes [Output Pose()]}
 EventGraph {nodes [
   TryGetPawnOwner()
   Event BlueprintUpdateAnimation()
  ]}
 MovementMode_AimOffset {nodes [
   Input Pose() {name=InAimOffsetsPose}
   Input Pose() {name=InSourcePosePreAimOffsets}
   Output Pose() {layerGroup=MovementMode}
  ]}
 MovementMode_BaseLocomotion {nodes [
   Input Pose() {name=InPoseBaseLocomotion}
   Output Pose() {layerGroup=MovementMode}
  ]}
 MovementMode_FinalPoseOverride {nodes [
   Input Pose() {name=InFinalPose}
   Output Pose() {layerGroup=MovementMode}
  ]}
 MovementMode_FullBodyOverride {nodes [
   Input Pose() {name=InputPoseFullBody}
   Output Pose() {layerGroup=MovementMode}
  ]}
 MovementMode_InAirOverride {nodes [
   Input Pose() {name=InSource_Pose}
   Input Pose() {name=InAirPose}
   n1 = Get DirectionToSequenceMap() -> (DirectionToSequenceMap:Map of EFortCardinalDirection Enums to Anim Sequence Object References=North)
   n2 = Get Direction() -> (Direction:EFortCardinalDirection Enum=North)
   Output Pose(Result=Sequence Player!(Sequence:Anim Sequence Base Object Reference=Find(TargetMap:Map of EFortCardinalDirection Enums to Anim Sequence Object References=@n1, Key:"EFortCardinalDirection Enum (by ref)"=@n2) -> (Value:Anim Sequence Object Reference)) -> (Pose)) {layerGroup=MovementMode}
  ]}
 MovementMode_LocomotionAdditives {nodes [
   Input Pose() {name=InLocomotionAdditivesPose}
   Output Pose() {layerGroup=MovementMode}
  ]}
 MovementMode_LowerBodyOverride {nodes [
   Input Pose() {name=InLowerBodyPose}
   Input Pose() {name=InSourcePose}
   Output Pose() {layerGroup=MovementMode}
  ]}
 MovementMode_OffScreen {nodes [
   Input Pose() {name=InPoseOffscreen}
   Output Pose() {layerGroup=MovementMode}
  ]}
 MovementMode_SourcePose {nodes [
   Input Pose() {name=InSourcePoseBase}
   Output Pose() {layerGroup=MovementMode}
  ]}
]
variables [
 Direction {type=EFortCardinalDirection, default=North}
 DirectionToSequenceMap {default {
   east=&/Boost_Jump_Right
   north=&/Boost_Jump_Forward
   south=&/Boost_Jump_Backward
   west=&/Boost_Jump_Left
  }}
]
