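# Boosted Jump in-air loop sound cue, with separate local, enemy and non-enemy mixes on a capped-concurrency player foley class.
# IsLocal picks the local mix or falls through to IsEnemy, where both remote mixes share jump attenuation and non-enemies play much quieter.
# Each mix layers one of three random start one-shots over a looping wave, whose Enveloper fades volume in and raises pitch over time.
# An unset IsEnemy gets the enemy mix.
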SoundCue /BoostedJump/Audio/SC_BoostedJump_InAir_Loop
root=Branch_0
nodes [
 {id=Attenuation_0, children [Mixer_0]}
 {id=Attenuation_1, children [Mixer_1], properties {attenuationSettings=&/ATT_Boosted_Jump}}
 {id=Attenuation_2, children [Mixer_2], properties {attenuationSettings=&ATT_Boosted_Jump}}
 {id=Branch_0, children [Attenuation_0, Branch_1, Branch_1], properties {boolParameterName=IsLocal}}
 {id=Branch_1, children [Mixer_3,Mixer_4,Mixer_3], properties {boolParameterName=IsEnemy}}
 {
  id=Enveloper_0
  children [Modulator_1]
  properties {
   pitchCurve {editorCurveData {keys [{arriveTangent=0.2634, interpMode=RCIM_Cubic, leaveTangent=0.2634, tangentMode=RCTM_User, time=0, value=0.8}, <at 5 1.75>]}}
   volumeCurve {editorCurveData {keys [<at 0 0 RCIM_Cubic>,<at 2 1>]}}
  }
 }
 {
  id=Enveloper_1
  children [Modulator_3]
  properties {
   pitchCurve {editorCurveData {keys [{arriveTangent=0.2634, interpMode=RCIM_Cubic, leaveTangent=0.2634, tangentMode=RCTM_User, time=0, value=0.8}, <at 5 1.75>]}}
   volumeCurve {editorCurveData {keys [<at 0 0 RCIM_Cubic>,<at 2 1>]}}
  }
 }
 {
  id=Enveloper_2
  children [Modulator_5]
  properties {
   pitchCurve {editorCurveData {keys [{arriveTangent=0.2634, interpMode=RCIM_Cubic, leaveTangent=0.2634, tangentMode=RCTM_User, time=0, value=0.8}, <at 5 1.75>]}}
   volumeCurve {editorCurveData {keys [<at 0 0 RCIM_Cubic>,<at 2 1>]}}
  }
 }
 {id=Mixer_0, children [Modulator_0,Enveloper_0], properties {inputVolume [0.5,1]}}
 {id=Mixer_1, children [Modulator_2,Enveloper_1], properties {inputVolume [0.6,1]}}
 {id=Mixer_2, children [Modulator_4,Enveloper_2], properties {inputVolume [0.3,1]}}
 {id=Mixer_3, children [Attenuation_1], properties {inputVolume [1]}}
 {id=Mixer_4, children [Attenuation_2], properties {inputVolume [0.25]}}
 {id=Modulator_0, children [Random_0], properties {volumeMax=1}}
 {id=Modulator_1, children [WavePlayer_0], properties {volumeMax=1}}
 {id=Modulator_2, children [Random_1], properties {volumeMax=1}}
 {id=Modulator_3, children [WavePlayer_5], properties {volumeMax=1}}
 {id=Modulator_4, children [Random_2], properties {volumeMax=1}}
 {id=Modulator_5, children [WavePlayer_9], properties {volumeMax=1}}
 {id=Random_0, children [WavePlayer_1,WavePlayer_2,WavePlayer_3], properties {bShouldExcludeFromBranchCulling=true, weights [1,1,1]}}
 {id=Random_1, children [WavePlayer_6,WavePlayer_7,WavePlayer_8], properties {bShouldExcludeFromBranchCulling=true, weights [1,1,1]}}
 {id=Random_2, children [WavePlayer_10, WavePlayer_11, WavePlayer_12], properties {bShouldExcludeFromBranchCulling=true, weights [1,1,1]}}
 {id=WavePlayer_0, properties {bLooping=true, soundWaveAssetPtr=&/BoostedJump_InAir_Loop_01}}
 {id=WavePlayer_1, properties {soundWaveAssetPtr=&/BoostedJump_InAir_Loop_Start_01}}
 {id=WavePlayer_10, properties {soundWaveAssetPtr=&BoostedJump_InAir_Loop_Start_01}}
 {id=WavePlayer_11, properties {soundWaveAssetPtr=&/BoostedJump_InAir_Loop_Start_02}}
 {id=WavePlayer_12, properties {soundWaveAssetPtr=&/BoostedJump_InAir_Loop_Start_03}}
 {id=WavePlayer_2, properties {soundWaveAssetPtr=&BoostedJump_InAir_Loop_Start_02}}
 {id=WavePlayer_3, properties {soundWaveAssetPtr=&BoostedJump_InAir_Loop_Start_03}}
 {id=WavePlayer_5, properties {bLooping=true, soundWaveAssetPtr=&BoostedJump_InAir_Loop_01}}
 {id=WavePlayer_6, properties {soundWaveAssetPtr=&BoostedJump_InAir_Loop_Start_01}}
 {id=WavePlayer_7, properties {soundWaveAssetPtr=&BoostedJump_InAir_Loop_Start_02}}
 {id=WavePlayer_8, properties {soundWaveAssetPtr=&BoostedJump_InAir_Loop_Start_03}}
 {id=WavePlayer_9, properties {bLooping=true, soundWaveAssetPtr=&BoostedJump_InAir_Loop_01}}
]
properties {
 bOverrideConcurrency=true
 maxDistance=2097152
 soundClassObject=&Fort_Player_Foley_Athena
 volumeMultiplier=1
 concurrencyOverrides {maxCount=3}
}
