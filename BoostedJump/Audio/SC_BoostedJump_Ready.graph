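# SC_BoostedJump_Ready is a SoundCue, named for the ready state of a boosted jump, with an empty node graph and no root, so it references no sound.
# It is configured for the Athena player foley sound class and takes spatialization settings from the ATT_Boosted_Jump attenuation asset.
# It overrides concurrency to cap simultaneous instances at two and disables virtualization.
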SoundCue /BoostedJump/Audio/SC_BoostedJump_Ready
root=null
nodes []
properties {
 attenuationSettings=&/ATT_Boosted_Jump
 bOverrideConcurrency=true
 maxDistance=5650
 soundClassObject=&Fort_Player_Foley_Athena
 virtualizationMode=Disabled
 volumeMultiplier=1
 concurrencyOverrides {maxCount=2}
}
